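# Looping thrust sound cue for the PrivateJet glider cosmetic.
# A Mixer blends six looping layers (wind, bells, flap, groan, texture and horse), each with its own input volume.
# Each layer passes through a continuous modulator driven by the AnyDirection parameter, shifting pitch and volume within per-layer ranges.
# Playback uses the player glider thrust attenuation, the glider sound class and a cosmetic glider loop properties sheet.
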SoundCue /BRCosmetics/Sounds/Gliders/PrivateJet/GliderThrustLoop_PrivateJet_Cue
root=Mixer_0
nodes [
 {
  id=Mixer_0
  children [ModulatorContinuous_0, ModulatorContinuous_1, ModulatorContinuous_2, ModulatorContinuous_3, ModulatorContinuous_4, ModulatorContinuous_5]
  properties {inputVolume [0.9,0.6,0.8,1,0.7,0.7]}
 }
 {
  id=ModulatorContinuous_0
  children [WavePlayer_0]
  properties {
   pitchModulationParams {default=0, maxOutput=1.1, minOutput=1, parameterName=AnyDirection}
   volumeModulationParams {default=0, minOutput=0.4, parameterName=AnyDirection}
  }
 }
 {
  id=ModulatorContinuous_1
  children [WavePlayer_1]
  properties {
   pitchModulationParams {default=0, maxOutput=1.02, minOutput=1, parameterName=AnyDirection}
   volumeModulationParams {default=0, maxOutput=0.5, minOutput=0.3, parameterName=AnyDirection}
  }
 }
 {
  id=ModulatorContinuous_2
  children [WavePlayer_2]
  properties {
   pitchModulationParams {default=0, maxOutput=1.1, minOutput=1, parameterName=AnyDirection}
   volumeModulationParams {default=0, maxOutput=0.8, minOutput=0.3, parameterName=AnyDirection}
  }
 }
 {
  id=ModulatorContinuous_3
  children [WavePlayer_3]
  properties {
   pitchModulationParams {default=0, maxOutput=1.2, minOutput=1, parameterName=AnyDirection}
   volumeModulationParams {default=0, minOutput=0.3, parameterName=AnyDirection}
  }
 }
 {
  id=ModulatorContinuous_4
  children [WavePlayer_4]
  properties {
   pitchModulationParams {default=0, maxOutput=1.02, minOutput=1, parameterName=AnyDirection}
   volumeModulationParams {default=0, maxOutput=0.8, minOutput=0.4, parameterName=AnyDirection}
  }
 }
 {
  id=ModulatorContinuous_5
  children [WavePlayer_5]
  properties {
   pitchModulationParams {default=0, minOutput=0.9, parameterName=AnyDirection}
   volumeModulationParams {default=0, minOutput=0.5, parameterName=AnyDirection}
  }
 }
 {id=WavePlayer_0, properties {bLooping=true, soundWaveAssetPtr=&/IP_Glider_PrivateJet_Thrust_LP_Wind}}
 {id=WavePlayer_1, properties {bLooping=true, soundWaveAssetPtr=&/IP_Glider_PrivateJet_Thrust_LP_Bells}}
 {id=WavePlayer_2, properties {bLooping=true, soundWaveAssetPtr=&/IP_Glider_PrivateJet_Thrust_LP_Flap}}
 {id=WavePlayer_3, properties {bLooping=true, soundWaveAssetPtr=&/IP_Glider_PrivateJet_Thrust_LP_Groan}}
 {id=WavePlayer_4, properties {bLooping=true, soundWaveAssetPtr=&/IP_Glider_PrivateJet_Thrust_LP_Texture}}
 {id=WavePlayer_5, properties {bLooping=true, soundWaveAssetPtr=&/IP_Glider_PrivateJet_Thrust_LP_Horse}}
]
properties {
 attenuationSettings=&Fort_Player_Glider_Thrust_Athena_Att
 audioPropertiesSheet=&APS_Cosmetic_Glider_Loop
 maxDistance=12500
 soundClassObject=&Fort_Gliders_Land
 volumeMultiplier=0.95
 concurrencySet [null]
}
